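; Gameplay ability for the Grind Rail Heal player augment, which switches on a regen effect while the player rides a grind rail or zipline.
; On event activation it checks whether healing is allowed, blocking it inside a storm unless a scalable-float toggle permits it.
; If allowed, it activates the augment and waits for the triggering event tag to be removed, then ends the effect.
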(event Ability|EventActivateAbilityFromEvent
  (params (EventData_EventTag :type Gameplay%20Tag%20Structure) (EventData_Instigator :type Actor%20Object%20Reference) (EventData_Target :type Actor%20Object%20Reference) (EventData_OptionalObject :type Object%20Reference) (EventData_OptionalObject2 :type Object%20Reference) (EventData_ContextHandle :type Gameplay%20Effect%20Context%20Handle%20Structure) (EventData_InstigatorTags :type Gameplay%20Tag%20Container%20Structure) (EventData_TargetTags :type Gameplay%20Tag%20Container%20Structure) (EventData_EventMagnitude :type Float%20%28single-precision%29) (EventData_TargetData :type Gameplay%20Ability%20Target%20Data%20Handle%20Structure) (EventData_InstancedEventData :type Instanced%20Struct%20Structure))
  ;[1 0.2718 0.06097] Activate when on grind rails or zipline
  (if (CallFunction|GetHealAllowed?)
    (CallFunction|ActivateAugment)
    (Ability|Tasks|WaitGameplayTagRemove EventData_EventTag 0 true
      (:Removed
        (CallFunction|EndAugmentEffect)))))

(fn GetHealAllowed?
  ;[0.583 0.2655 1] Do not heal if inside the storm
  (return (select (ScalableFloat|GetValueasBool (Variables|Default|GetHealableInStorm?)) true (not (Variables|/Script/FortniteGame.FortPlayerPawn|GetbIsInAnyStorm (CallFunction|GetAvatarasFortPlayerPawn))))))
